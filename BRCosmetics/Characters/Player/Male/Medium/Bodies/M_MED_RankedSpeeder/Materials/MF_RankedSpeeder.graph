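# Material function that applies the ranked thin-film cosmetic look to the RankedSpeeder character body and helmet, outputting material attributes and a MaskFX mask.
# RankedMask texture channels, plus an IconMask input on non-helmet parts, drive base color, metallic and specular blends.
# The result feeds MF_RankedThinFilm, then base color gets hue shift and desaturation.
# The IsHelmet static switch selects masks and adds SphereMask-keyed detail darkening and masked roughness lift.
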MaterialFunction /BRCosmetics/Characters/Player/Male/Medium/Bodies/M_MED_RankedSpeeder/Materials/MF_RankedSpeeder
graph {nodes [
  FunctionOutput(None=NamedRerouteUsage()) {outputName=MaskFX, sortPriority=1}
  n1 = NamedRerouteUsage()
  n2 = ScalarParameter() {defaultValue=1, group=RankedSpeeder, parameterName=BaseColorDesaturation}
  n3 = TextureSampleParameter2D() {
   group=RankedSpeeder
   parameterName=RankedMask
   samplerSource=SSM_Wrap_WorldGroupSettings
   texture=/BRCosmetics/Characters/Player/Male/Medium/Bodies/M_MED_RankedSpeeder/Textures/T_RankedSpeeder_body_RM_FX
  }
  n4 = Add(A=@n3.R, B=@n3.R)
  n5 = Add(A=@n3.G, B=@n3.G)
  n6 = LinearInterpolate(
   A=NamedRerouteUsage()
   B=Multiply(A=Desaturation(None=@n1, Fraction=@n2), B=ScalarParameter() {defaultValue=0.2, group=RankedSpeeder, parameterName=BaseColorMultiplier}) {constB=0.2}
   Alpha=StaticSwitchParameter(True=Saturate(None=@n4), False=Saturate(None=@n5)) {group=RankedSpeeder, parameterName=IsHelmet}
  )
  n7 = LinearInterpolate(A=NamedRerouteUsage(), Alpha=NamedRerouteUsage()) {constB=0.9}
  n8 = LinearInterpolate(A=NamedRerouteUsage(), Alpha=NamedRerouteUsage()) {constB=0.8}
  n9 = GetMaterialAttributes(None=FunctionInput() {inputName=MaterialAttributes, inputType=FunctionInput_MaterialAttributes}) {attributeGetTypes [
    #1
    #2
    #3
    #4
   ]}
  n10 = GetMaterialAttributes(None=MaterialFunctionCall!(
    Material Attributes=SetMaterialAttributes!(MaterialAttributes=@n9.MaterialAttributes, BaseColor=@n6, Metallic=@n7, Specular=@n8) {attributeSetTypes [#1, #2, #3]}
    Alpha=NamedRerouteUsage()
   ) {materialFunction=/Game/Global/Materials/Cosmetics/Ranked/MF_RankedThinFilm}.Result) {attributeGetTypes [#1]}
  n11 = Add(
   A=NamedRerouteUsage()
   B=SphereMask(A=NamedRerouteUsage(), B=Constant3Vector() {constant=<rgba 0.117 0.0012 0.008 1>}) {attenuationRadius=0.05}
  )
  n12 = ScalarParameter() {defaultValue=0.5, group=RankedSpeeder, parameterName=HelmetDetailIntensity}
  n13 = StaticSwitchParameter(True=OneMinus(None=Multiply(A=@n11, B=@n12) {constB=0.8}), False=Constant() {r=1}) {group=RankedSpeeder, parameterName=IsHelmet}
  n14 = Desaturation(
   None=MaterialFunctionCall(Hue Shift Percentage=ScalarParameter() {group=RankedSpeeder, parameterName=ThinFilm_HueShift, sortPriority=20}, Texture=@n10.BaseColor) {materialFunction=&HueShift}.Result
   Fraction=ScalarParameter() {group=RankedSpeeder, parameterName=ThinFilm_Desaturation, sortPriority=20}
  )
  n15 = NamedRerouteUsage()
  n16 = LinearInterpolate(A=@n15, B=Saturate(None=Add(A=@n15) {constB=0.23}), Alpha=NamedRerouteUsage())
  FunctionOutput(None=SetMaterialAttributes!(
    MaterialAttributes=@n10.MaterialAttributes
    BaseColor=Multiply(A=@n13, B=@n14)
    Roughness=StaticSwitchParameter(True=@n16, False=@n15) {group=RankedSpeeder, parameterName=IsHelmet}
   ) {attributeSetTypes [#1, #4]})
  NamedRerouteDeclaration(None=@n9.BaseColor) {name=OriginalBaseColor, nodeColor=<rgba 0.9219 1 0 1>}
  NamedRerouteDeclaration(None=@n9.Metallic) {name=OriginalMetallic, nodeColor=<rgba 0.7578 0 1 1>}
  NamedRerouteDeclaration(None=@n9.Specular) {name=OriginalSpecular, nodeColor=<rgba 0 1 0.4375 1>}
  n17 = Add(A=NamedRerouteDeclaration(None=@n3.R) {name=HelmetYellowLineMask, nodeColor=<rgba 1 0.8648 0.1491 1>}, B=@n3.G)
  n18 = Add(A=@n3.G, B=FunctionInput() {inputName=IconMask, inputType=FunctionInput_Scalar, sortPriority=10})
  NamedRerouteDeclaration(None=StaticSwitchParameter(True=Saturate(None=@n17), False=Saturate(None=@n18)) {group=RankedSpeeder, parameterName=IsHelmet}) {name=RankedMask, nodeColor=<rgba 1 0.1172 0 1>}
  NamedRerouteDeclaration(None=MaterialFunctionCall(R=@n3.R, G=NamedRerouteUsage(), B=@n3.R) {materialFunction=&AppendMany}.RGB) {name=MaskFX, nodeColor=<rgba 1 0.5156 0 1>}
  NamedRerouteDeclaration(None=@n9.Roughness) {name=OriginalRoughness, nodeColor=<rgba 0 0.8359 1 1>}
  Reroute(None=@n3.R)
 ]}
