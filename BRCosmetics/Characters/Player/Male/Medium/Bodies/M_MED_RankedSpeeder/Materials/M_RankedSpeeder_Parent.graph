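# Parent material for the male medium RankedSpeeder player cosmetic body, layering outfit-specific effects onto the shared 2023 character material.
# Material attributes flow from the character PreFX function through MF_Icon_RankedFury into MF_RankedSpeeder, which also receives the icon mask.
# An emissive fresnel function then uses the PreFX fresnel and the green channel of MaskFX before the shared PostFX function.
# The shading model is set by material expression.
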Material /BRCosmetics/Characters/Player/Male/Medium/Bodies/M_MED_RankedSpeeder/Materials/M_RankedSpeeder_Parent
graph {
 properties {
  bUseMaterialAttributes=true
  bUsedWithInstancedStaticMeshes=true
  bUsedWithMeshDeformer=true
  bUsedWithMorphTargets=true
  bUsedWithNanite=true
  bUsedWithSkeletalMesh=true
  pixelDepthOffsetMode=PDOM_Legacy
  shadingModel=MSM_FromMaterialExpression
  usedShadingModels=Default Lit | Subsurface | Clear Coat | Subsurface Profile | Cloth
 }
 nodes [
  n1 = MaterialFunctionCall() {materialFunction=&MF_CharacterParent_2023_PreFX}
  n2 = MaterialFunctionCall!(Material=@n1.Result, Alpha=@n1.Result) {materialFunction=&/MF_Icon_RankedFury}
  n3 = MaterialFunctionCall!(MaterialAttributes=@n2.Result, IconMask=@n2.Result) {materialFunction=&/MF_RankedSpeeder}
  MaterialFunctionCall(
   Material=MaterialFunctionCall!(Fresnel=@n1.Fresnel, Mask=ComponentMask(None=@n3.MaskFX) {g=true}, MaterialAttributes=@n3.Result) {materialFunction=&/MF_RankedSpeeder_EmissiveFresnel}.SetMaterialAttributes
   M=@n1.M
   Pre Skinned Local Position=@n1.M
   Distance Blend=@n1.M
   Fresnel=@n1.M
  ) {materialFunction=&MF_CharacterParent_2023_PostFX}
 ]
}
