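# Physics asset for the male medium RankedSpeeder player body, providing capsule bodies for the character and the dangling parts of its outfit.
# Torso, arm, hand and leg capsules are joined from the pelvis outward by uniform angular constraints, with no head or foot bodies.
# Jacket, chest-flap and front-collar dyn_ bones form short, tightly limited chains from tiny base_ anchor capsules, which attach to spine_04 or spine_05 with free swing and twist.
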PhysicsAsset /BRCosmetics/Characters/Player/Male/Medium/Bodies/M_MED_RankedSpeeder/Meshes/M_MED_RankedSpeeder_Physics
bodies {
 base_chestflap_l [{shapeType=Capsule, length=0.505, radius=0.505, shapeName=base_chestflap_l_capsule, rotation=<rot 0 -90 0>}]
 base_chestflap_r [{shapeType=Capsule, length=0.505, radius=0.505, shapeName=base_chestflap_r_capsule, rotation=<rot 0 -90 0>}]
 base_collar_front_01_l [{shapeType=Capsule, length=0.505, radius=0.505, shapeName=base_collar_front_01_l_capsule, rotation=<rot 0 -90 0>}]
 base_collar_front_01_r [{shapeType=Capsule, length=0.505, radius=0.505, shapeName=base_collar_front_01_r_capsule, rotation=<rot 0 -90 0>}]
 base_jacket_lower_01_l [{shapeType=Capsule, length=0.505, radius=0.505, shapeName=base_jacket_lower_01_l_capsule, rotation=<rot 0 -90 0>}]
 base_jacket_lower_01_r [{shapeType=Capsule, length=0.505, radius=0.505, shapeName=base_jacket_lower_01_r_capsule, rotation=<rot 0 -90 0>}]
 base_jacket_mid [{shapeType=Capsule, length=0.505, radius=0.505, shapeName=base_jacket_mid_capsule, rotation=<rot 0 -90 0>}]
 base_jacket_side_01_l [{shapeType=Capsule, length=0.505, radius=0.505, shapeName=base_jacket_side_01_l_capsule, rotation=<rot 0 -90 0>}]
 base_jacket_side_01_r [{shapeType=Capsule, length=0.505, radius=0.505, shapeName=base_jacket_side_01_r_capsule, rotation=<rot 0 -90 0>}]
 calf_l [{shapeType=Capsule, length=34.2047, radius=9.5383, center=<-24.3123 -1.0713 -0.3802>, rotation=<rot -0.0236 86.9986 -89.7301>}]
 calf_r [{shapeType=Capsule, length=34.2047, radius=9.5383, center=<24.4116 0.8213 -0.381>, rotation=<rot -0.0236 86.9986 -89.7301>}]
 dyn_chestflap_01_l [{shapeType=Capsule, length=2.3447, radius=1.0828, shapeName=dyn_chestflap_01_l_capsule, center=<1.7921 0 0>, rotation=<rot 0 90 90>}]
 dyn_chestflap_01_r [{shapeType=Capsule, length=2.3447, radius=1.0828, shapeName=dyn_chestflap_01_l_capsule, center=<1.7921 0 0>, rotation=<rot 0 90 90>}]
 dyn_chestflap_02_l [{shapeType=Capsule, length=2.5755, radius=1.37, shapeName=dyn_chestflap_02_l_capsule, center=<0.9873 0 0>, rotation=<rot 0 -90 90>}]
 dyn_chestflap_02_r [{shapeType=Capsule, length=2.5755, radius=1.37, shapeName=dyn_chestflap_02_l_capsule, center=<0.9873 0 0>, rotation=<rot 0 -90 90>}]
 dyn_collar_front_01_l [{shapeType=Capsule, length=1.005, radius=1.005, shapeName=dyn_collar_front_01_l_capsule, center=<0.6457 0 0>, rotation=<rot 0 -90 90>}]
 dyn_collar_front_01_r [{shapeType=Capsule, length=1.005, radius=1.005, shapeName=dyn_collar_front_01_l_capsule, center=<0.6457 0 0>, rotation=<rot 0 -90 90>}]
 dyn_collar_front_02_l [{shapeType=Capsule, length=1.767, radius=1.0356, shapeName=dyn_collar_front_02_l_capsule, center=<1.063 0 0>, rotation=<rot 0 -90 90>}]
 dyn_collar_front_02_r [{shapeType=Capsule, length=1.767, radius=1.0356, shapeName=dyn_collar_front_02_l_capsule, center=<1.063 0 0>, rotation=<rot 0 -90 90>}]
 dyn_jacket_lower_01_l [{shapeType=Capsule, length=2.8932, radius=1.1404, shapeName=dyn_jacket_lower_01_l_capsule, center=<1.9434 0 0>, rotation=<rot 0 -90.0004 -90>}]
 dyn_jacket_lower_01_r [{shapeType=Capsule, length=2.8932, radius=1.1404, shapeName=dyn_jacket_lower_01_l_capsule, center=<1.9434 0 0>, rotation=<rot 0 -90.0004 -90>}]
 dyn_jacket_lower_02_l [{shapeType=Capsule, length=5.3859, radius=2.2, shapeName=dyn_jacket_lower_02_l_capsule, center=<2.2077 0 0>, rotation=<rot 0 90 90>}]
 dyn_jacket_lower_02_r [{shapeType=Capsule, length=5.0303, radius=2.2, shapeName=dyn_jacket_lower_02_l_capsule, center=<2.2077 0 0>, rotation=<rot 0 90 90>}]
 dyn_jacket_mid_01 [{shapeType=Capsule, length=6.8106, radius=1.4864, shapeName=dyn_jacket_mid_01_capsule, center=<4.5103 0 0>, rotation=<rot 0 90 -90>}]
 dyn_jacket_side_01_l [{shapeType=Capsule, length=6.8106, radius=1.5, shapeName=dyn_jacket_mid_01_capsule, center=<5.0662 0 0>, rotation=<rot 0 80.0005 -90>}]
 dyn_jacket_side_01_r [{shapeType=Capsule, length=6.8106, radius=1.5, shapeName=dyn_jacket_mid_01_capsule, center=<5.0662 0 0>, rotation=<rot 0 80.0005 -90>}]
 hand_l [{shapeType=Capsule, length=10.222, radius=5.9862, center=<7.0608 -2.2193 -0.4693>, rotation=<rot 0 78.8353 -86.3135>}]
 hand_r [{shapeType=Capsule, length=10.2153, radius=5.9966, center=<-7.0641 2.227 0.4626>, rotation=<rot 0 79.1718 -86.5132>}]
 lowerarm_l [{shapeType=Capsule, length=12.2597, radius=7.4732, center=<8.4629 1.1168 1.928>, rotation=<rot 0 -92.7685 -84.3037>}]
 lowerarm_r [{shapeType=Capsule, length=12.2693, radius=7.4477, center=<-8.461 -1.115 -1.9228>, rotation=<rot 0 -92.7808 -84.5442>}]
 pelvis [{shapeType=Capsule, length=10.3736, radius=10.6713, center=<-1.0437 -0.1099 -0.0354>, rotation=<rot 0.8919 -90 -0.6118>}]
 spine_01 [{shapeType=Capsule, length=8.0108, radius=8.8278, center=<6.2088 0.3526 0.1025>, rotation=<rot -2.0422 -90 -1.4583>}]
 spine_03 [{shapeType=Capsule, length=20.935, radius=7.0189, center=<9.5592 2.9419 0.8276>, rotation=<rot 0 176.9362 0.8886>}]
 spine_04 [{shapeType=Capsule, length=9.6362, radius=8.7323, center=<-3.8809 0.7849 -0.0619>, rotation=<rot -0.2248 -90 -0.3861>}]
 spine_05 [{shapeType=Capsule, length=25.8663, radius=11.6475, center=<1.713 -0.1631 0.0962>, rotation=<rot -0.4563 -90 0.4079>}]
 thigh_l [{shapeType=Capsule, length=23.9062, radius=10.6443, center=<-16.0185 0.5437 -0.4325>, rotation=<rot 0 92.1536 -93.3274>}]
 thigh_r [{shapeType=Capsule, length=23.9062, radius=10.6443, center=<15.9275 -1.6853 -0.4325>, rotation=<rot 0 87.5806 -93.2164>}]
 upperarm_l [{shapeType=Capsule, length=13.899, radius=6.0445, center=<13.0487 -1.1125 -0.9672>, rotation=<rot -2.6853 96.1887 -84.5716>}]
 upperarm_r [{shapeType=Capsule, length=13.899, radius=6.0445, center=<-13.0487 -1.1125 -0.9672>, rotation=<rot -2.6853 96.1887 -84.5716>}]
}
constraints [
 {bone1Name=spine_01, bone2Name=pelvis, swing1LimitDegrees=45, swing2LimitDegrees=45, twistLimitDegrees=45}
 {bone1Name=spine_03, bone2Name=spine_01, swing1LimitDegrees=45, swing2LimitDegrees=45, twistLimitDegrees=45}
 {bone1Name=spine_04, bone2Name=spine_03, swing1LimitDegrees=45, swing2LimitDegrees=45, twistLimitDegrees=45}
 {bone1Name=spine_05, bone2Name=spine_04, swing1LimitDegrees=45, swing2LimitDegrees=45, twistLimitDegrees=45}
 {bone1Name=upperarm_l, bone2Name=spine_05, swing1LimitDegrees=45, swing2LimitDegrees=45, twistLimitDegrees=45}
 {bone1Name=lowerarm_l, bone2Name=upperarm_l, swing1LimitDegrees=45, swing2LimitDegrees=45, twistLimitDegrees=45}
 {bone1Name=hand_l, bone2Name=lowerarm_l, swing1LimitDegrees=45, swing2LimitDegrees=45, twistLimitDegrees=45}
 {bone1Name=upperarm_r, bone2Name=spine_05, swing1LimitDegrees=45, swing2LimitDegrees=45, twistLimitDegrees=45}
 {bone1Name=lowerarm_r, bone2Name=upperarm_r, swing1LimitDegrees=45, swing2LimitDegrees=45, twistLimitDegrees=45}
 {bone1Name=hand_r, bone2Name=lowerarm_r, swing1LimitDegrees=45, swing2LimitDegrees=45, twistLimitDegrees=45}
 {bone1Name=thigh_l, bone2Name=pelvis, swing1LimitDegrees=45, swing2LimitDegrees=45, twistLimitDegrees=45}
 {bone1Name=calf_l, bone2Name=thigh_l, swing1LimitDegrees=45, swing2LimitDegrees=45, twistLimitDegrees=45}
 {bone1Name=thigh_r, bone2Name=pelvis, swing1LimitDegrees=45, swing2LimitDegrees=45, twistLimitDegrees=45}
 {bone1Name=calf_r, bone2Name=thigh_r, swing1LimitDegrees=45, swing2LimitDegrees=45, twistLimitDegrees=45}
 {bone1Name=dyn_jacket_lower_01_l, bone2Name=base_jacket_lower_01_l, swing1LimitDegrees=25, swing2LimitDegrees=5, twistLimitDegrees=5}
 {bone1Name=dyn_jacket_lower_02_l, bone2Name=dyn_jacket_lower_01_l, swing1LimitDegrees=25, swing2LimitDegrees=5, twistLimitDegrees=5}
 {bone1Name=dyn_jacket_lower_01_r, bone2Name=base_jacket_lower_01_r, swing1LimitDegrees=25, swing2LimitDegrees=5, twistLimitDegrees=5}
 {bone1Name=dyn_jacket_lower_02_r, bone2Name=dyn_jacket_lower_01_r, swing1LimitDegrees=25, swing2LimitDegrees=5, twistLimitDegrees=5}
 {bone1Name=dyn_jacket_mid_01, bone2Name=base_jacket_mid, swing1LimitDegrees=18, swing2LimitDegrees=5, twistLimitDegrees=5}
 {bone1Name=dyn_jacket_side_01_l, bone2Name=base_jacket_side_01_l, swing1LimitDegrees=12, swing2LimitDegrees=5, twistLimitDegrees=5}
 {bone1Name=dyn_jacket_side_01_r, bone2Name=base_jacket_side_01_r, swing1LimitDegrees=12, swing2LimitDegrees=5, twistLimitDegrees=5}
 {bone1Name=dyn_chestflap_01_l, bone2Name=base_chestflap_l, swing1LimitDegrees=25, swing2LimitDegrees=5, twistLimitDegrees=5}
 {bone1Name=dyn_chestflap_02_l, bone2Name=dyn_chestflap_01_l, swing1LimitDegrees=25, swing2LimitDegrees=5, twistLimitDegrees=5}
 {bone1Name=dyn_chestflap_01_r, bone2Name=base_chestflap_r, swing1LimitDegrees=25, swing2LimitDegrees=5, twistLimitDegrees=5}
 {bone1Name=dyn_chestflap_02_r, bone2Name=dyn_chestflap_01_r, swing1LimitDegrees=25, swing2LimitDegrees=5, twistLimitDegrees=5}
 {bone1Name=dyn_collar_front_01_l, bone2Name=base_collar_front_01_l, swing1LimitDegrees=25, swing2LimitDegrees=5, twistLimitDegrees=5}
 {bone1Name=dyn_collar_front_02_l, bone2Name=dyn_collar_front_01_l, swing1LimitDegrees=25, swing2LimitDegrees=5, twistLimitDegrees=5}
 {bone1Name=dyn_collar_front_01_r, bone2Name=base_collar_front_01_r, swing1LimitDegrees=25, swing2LimitDegrees=5, twistLimitDegrees=5}
 {bone1Name=dyn_collar_front_02_r, bone2Name=dyn_collar_front_01_r, swing1LimitDegrees=25, swing2LimitDegrees=5, twistLimitDegrees=5}
 {
  bone1Name=base_jacket_lower_01_l
  bone2Name=spine_04
  swing1LimitDegrees=45
  swing1Motion=Free
  swing2LimitDegrees=45
  swing2Motion=Free
  twistLimitDegrees=45
  twistMotion=Free
 }
 {
  bone1Name=base_jacket_lower_01_r
  bone2Name=spine_04
  swing1LimitDegrees=45
  swing1Motion=Free
  swing2LimitDegrees=45
  swing2Motion=Free
  twistLimitDegrees=45
  twistMotion=Free
 }
 {
  bone1Name=base_jacket_mid
  bone2Name=spine_04
  swing1LimitDegrees=45
  swing1Motion=Free
  swing2LimitDegrees=45
  swing2Motion=Free
  twistLimitDegrees=45
  twistMotion=Free
 }
 {
  bone1Name=base_jacket_side_01_l
  bone2Name=spine_04
  swing1LimitDegrees=45
  swing1Motion=Free
  swing2LimitDegrees=45
  swing2Motion=Free
  twistLimitDegrees=45
  twistMotion=Free
 }
 {
  bone1Name=base_jacket_side_01_r
  bone2Name=spine_04
  swing1LimitDegrees=45
  swing1Motion=Free
  swing2LimitDegrees=45
  swing2Motion=Free
  twistLimitDegrees=45
  twistMotion=Free
 }
 {
  bone1Name=base_chestflap_l
  bone2Name=spine_05
  swing1LimitDegrees=45
  swing1Motion=Free
  swing2LimitDegrees=45
  swing2Motion=Free
  twistLimitDegrees=45
  twistMotion=Free
 }
 {
  bone1Name=base_chestflap_r
  bone2Name=spine_05
  swing1LimitDegrees=45
  swing1Motion=Free
  swing2LimitDegrees=45
  swing2Motion=Free
  twistLimitDegrees=45
  twistMotion=Free
 }
 {
  bone1Name=base_collar_front_01_l
  bone2Name=spine_05
  swing1LimitDegrees=45
  swing1Motion=Free
  swing2LimitDegrees=45
  swing2Motion=Free
  twistLimitDegrees=45
  twistMotion=Free
 }
 {
  bone1Name=base_collar_front_01_r
  bone2Name=spine_05
  swing1LimitDegrees=45
  swing1Motion=Free
  swing2LimitDegrees=45
  swing2Motion=Free
  twistLimitDegrees=45
  twistMotion=Free
 }
]
