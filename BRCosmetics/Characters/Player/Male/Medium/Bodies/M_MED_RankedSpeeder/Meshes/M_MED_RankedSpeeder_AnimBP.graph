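# Animation Blueprint for the M_MED_RankedSpeeder male body part, correcting collar, bag and leg pad bones on the player's pose.
# The pose copied from the parent mesh runs through a deform Control Rig, bone offsets and RigidBody physics into a cached pose.
# Bool blends select additive collar and bag corrections by emote, crouch, aim, movement and skydive state.
# The event graph derives their alphas, reading head and thigh bone deltas for head-up and clipping.
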AnimBlueprint /BRCosmetics/Characters/Player/Male/Medium/Bodies/M_MED_RankedSpeeder/Meshes/M_MED_RankedSpeeder_AnimBP
target_skeleton=&/M_MED_RankedSpeeder_Skeleton
graphs [
 AnimGraph {nodes [
   "Transform (Modify) Bone - Bone: base_bag"(Alpha=1, Translation="6,0,0", Rotation="-10.000,0.045,-0.016", Scale="1,1,1") {
    rotationMode=BMM_Additive
    rotationSpace=BCS_BoneSpace
    scaleMode=BMM_Additive
    scaleSpace=BCS_BoneSpace
    translationMode=BMM_Additive
    translationSpace=BCS_BoneSpace
    alphaBoolBlend {blendOption=Linear}
    alphaScaleBiasClamp {interpSpeedDecreasing=19.75, interpSpeedIncreasing=19.75}
    boneToModify {boneName=base_bag}
    rotation=<rot -10 0.045 -0.016>
    translation {x=6}
   }
   n1 = Get isTargetingHeadUp() -> (isTargetingHeadUp:double)
   n2 = "Transform (Modify) Bone - Bone: base_collar_mid_01_l"(
    ComponentPose=Local To Component(LocalPose=Use cached pose 'CTRLRigPose'())
    Alpha=@n1
    Translation="-5,-20,0"
    Rotation="23.144,-54.677,-15.330"
    Scale="1,1,1"
   ) {
    rotationMode=BMM_Additive
    rotationSpace=BCS_BoneSpace
    scaleMode=BMM_Additive
    scaleSpace=BCS_BoneSpace
    translationMode=BMM_Additive
    translationSpace=BCS_BoneSpace
    alphaBoolBlend {blendOption=Linear}
    alphaScaleBiasClamp {interpSpeedDecreasing=19.75, interpSpeedIncreasing=19.75}
    boneToModify {boneName=base_collar_mid_01_l}
    rotation=<rot 23.144 -54.677 -15.33>
    translation=<-5 -20>
   }
   n3 = "Transform (Modify) Bone - Bone: base_collar_side_01_l"(ComponentPose=@n2, Alpha=@n1, Translation="0,0,0", Rotation="0,0,0", Scale="1,1,1") {
    rotationMode=BMM_Additive
    rotationSpace=BCS_BoneSpace
    scaleMode=BMM_Additive
    scaleSpace=BCS_BoneSpace
    translationMode=BMM_Additive
    translationSpace=BCS_BoneSpace
    alphaBoolBlend {blendOption=Linear}
    alphaScaleBiasClamp {interpSpeedDecreasing=19.75, interpSpeedIncreasing=19.75}
    boneToModify {boneName=base_collar_side_01_l}
   }
   n4 = "Transform (Modify) Bone - Bone: base_collar_side_01_r"(ComponentPose=@n3, Alpha=@n1, Translation="-5,-20,0", Rotation="-0.230,-70.002,0.398", Scale="1,1,1") {
    rotationMode=BMM_Additive
    rotationSpace=BCS_BoneSpace
    scaleMode=BMM_Additive
    scaleSpace=BCS_BoneSpace
    translationMode=BMM_Additive
    translationSpace=BCS_BoneSpace
    alphaBoolBlend {blendOption=Linear}
    alphaScaleBiasClamp {interpSpeedDecreasing=19.75, interpSpeedIncreasing=19.75}
    boneToModify {boneName=base_collar_side_01_r}
    rotation=<rot -0.23 -70.002 0.398>
    translation=<-5 -20>
   }
   n5 = "Transform (Modify) Bone - Bone: base_bag"(
    ComponentPose=Local To Component(LocalPose=Use cached pose 'CTRLRigPose'())
    Alpha=Get IsCrouchingBag() -> (IsCrouchingBag:double)
    Translation="6,0,0"
    Rotation="-10.000,0.045,-0.016"
    Scale="1,1,1"
   ) {
    rotationMode=BMM_Additive
    rotationSpace=BCS_BoneSpace
    scaleMode=BMM_Additive
    scaleSpace=BCS_BoneSpace
    translationMode=BMM_Additive
    translationSpace=BCS_BoneSpace
    alphaBoolBlend {blendOption=Linear}
    alphaScaleBiasClamp {interpSpeedDecreasing=19.75, interpSpeedIncreasing=19.75}
    boneToModify {boneName=base_bag}
    rotation=<rot -10 0.045 -0.016>
    translation {x=6}
   }
   n6 = Get IsCrouchingheadup() -> (IsCrouchingheadup:double)
   n7 = "Transform (Modify) Bone - Bone: base_collar_mid_01_l"(ComponentPose=@n5, Alpha=@n6, Translation="-11.738,-9.612,0.354", Rotation="-0.022,-125.000,0.007", Scale="1,1,1") {
    rotationMode=BMM_Additive
    rotationSpace=BCS_BoneSpace
    scaleMode=BMM_Additive
    scaleSpace=BCS_BoneSpace
    translationMode=BMM_Additive
    translationSpace=BCS_BoneSpace
    alphaBoolBlend {blendOption=Linear}
    alphaScaleBiasClamp {interpSpeedDecreasing=19.75, interpSpeedIncreasing=19.75}
    boneToModify {boneName=base_collar_mid_01_l}
    rotation=<rot -0.022 -125 0.007>
    translation=<-11.738 -9.612 0.354>
   }
   n8 = "Transform (Modify) Bone - Bone: base_collar_side_01_l"(ComponentPose=@n7, Alpha=@n6, Translation="1.098,-8.785,5.530", Rotation="0.039,-75.000,0.023", Scale="1,1,1") {
    rotationMode=BMM_Additive
    rotationSpace=BCS_BoneSpace
    scaleMode=BMM_Additive
    scaleSpace=BCS_BoneSpace
    translationMode=BMM_Additive
    translationSpace=BCS_BoneSpace
    alphaBoolBlend {blendOption=Linear}
    alphaScaleBiasClamp {interpSpeedDecreasing=19.75, interpSpeedIncreasing=19.75}
    boneToModify {boneName=base_collar_side_01_l}
    rotation=<rot 0.039 -75 0.0229>
    translation=<1.098 -8.785 5.53>
   }
   n9 = "Transform (Modify) Bone - Bone: base_collar_side_01_r"(ComponentPose=@n8, Alpha=@n6, Translation="-4.290,-14.774,-0.579", Rotation="-0.131,-85.000,0.005", Scale="1,1,1") {
    rotationMode=BMM_Additive
    rotationSpace=BCS_BoneSpace
    scaleMode=BMM_Additive
    scaleSpace=BCS_BoneSpace
    translationMode=BMM_Additive
    translationSpace=BCS_BoneSpace
    alphaBoolBlend {blendOption=Linear}
    alphaScaleBiasClamp {interpSpeedDecreasing=19.75, interpSpeedIncreasing=19.75}
    boneToModify {boneName=base_collar_side_01_r}
    rotation=<rot -0.131 -85.0001 0.005>
    translation=<-4.29 -14.774 -0.579>
   }
   n10 = Blend Poses by bool(
    BlendPose_0=Component To Local(ComponentPose=@n9)
    BlendPose_1=Blend Poses by bool(
     BlendPose_0=Component To Local(ComponentPose=@n4)
     BlendPose_1=Use cached pose 'HeadUpPoses'()
     BlendTime_0=0.2
     bActiveValue=Get bIsTargeting()
    ) {blendType=Linear}
    BlendTime_0=0.2
    BlendTime_1=0.2
    bActiveValue=AND Boolean!(A=Get bIsCrouching(), B=Get bIsTargeting(), C:bool=Get bIsMoving2D()) -> (ReturnValue)
   ) {blendType=Linear}
   n11 = "Transform (Modify) Bone - Bone: base_collar_mid_01_l"(
    ComponentPose=Local To Component(LocalPose=Use cached pose 'CTRLRigPose'())
    Alpha=1
    Translation="-2,-8,0"
    Rotation="0,-30,0"
    Scale="1,1,1"
   ) {
    rotationMode=BMM_Additive
    rotationSpace=BCS_BoneSpace
    scaleMode=BMM_Additive
    scaleSpace=BCS_BoneSpace
    translationMode=BMM_Additive
    translationSpace=BCS_BoneSpace
    alphaBoolBlend {blendOption=Linear}
    alphaScaleBiasClamp {interpSpeedDecreasing=19.75, interpSpeedIncreasing=19.75}
    boneToModify {boneName=base_collar_mid_01_l}
    rotation {yaw=-30}
    translation=<-2 -8>
   }
   n12 = "Transform (Modify) Bone - Bone: base_collar_side_01_l"(ComponentPose=@n11, Alpha=1, Translation="-2,-8,0", Rotation="0,0,0", Scale="1,1,1") {
    rotationMode=BMM_Additive
    rotationSpace=BCS_BoneSpace
    scaleMode=BMM_Additive
    scaleSpace=BCS_BoneSpace
    translationMode=BMM_Additive
    translationSpace=BCS_BoneSpace
    alphaBoolBlend {blendOption=Linear}
    alphaScaleBiasClamp {interpSpeedDecreasing=19.75, interpSpeedIncreasing=19.75}
    boneToModify {boneName=base_collar_side_01_l}
    translation=<-2 -8>
   }
   n13 = "Transform (Modify) Bone - Bone: base_collar_side_01_r"(ComponentPose=@n12, Alpha=1, Translation="-2,-8,0", Rotation="0,0,0", Scale="1,1,1") {
    rotationMode=BMM_Additive
    rotationSpace=BCS_BoneSpace
    scaleMode=BMM_Additive
    scaleSpace=BCS_BoneSpace
    translationMode=BMM_Additive
    translationSpace=BCS_BoneSpace
    alphaBoolBlend {blendOption=Linear}
    alphaScaleBiasClamp {interpSpeedDecreasing=19.75, interpSpeedIncreasing=19.75}
    boneToModify {boneName=base_collar_side_01_r}
    translation=<-2 -8>
   }
   n14 = Component To Local(ComponentPose=@n13)
   n15 = Get bIsSkyDiving()
   n16 = "Transform (Modify) Bone - Bone: base_bag"(
    ComponentPose=Local To Component(LocalPose=Use cached pose 'CTRLRigPose'())
    Alpha=Get Left() -> (Left:double)
    Translation="3,0,0"
    Rotation="-10.000,0.045,-0.016"
    Scale="1,1,1"
   ) {
    rotationMode=BMM_Additive
    rotationSpace=BCS_BoneSpace
    scaleMode=BMM_Additive
    scaleSpace=BCS_BoneSpace
    translationMode=BMM_Additive
    translationSpace=BCS_BoneSpace
    alphaBoolBlend {blendOption=Linear}
    alphaScaleBiasClamp {interpSpeedDecreasing=19.75, interpSpeedIncreasing=19.75}
    boneToModify {boneName=base_bag}
    rotation=<rot -10 0.045 -0.016>
    translation {x=3}
   }
   n17 = "Transform (Modify) Bone - Bone: base_bag"(ComponentPose=@n16, Alpha=Get Right() -> (Right:double), Translation="3.5,0,0", Rotation="20,0,0", Scale="1,1,1") {
    rotationMode=BMM_Additive
    rotationSpace=BCS_BoneSpace
    scaleMode=BMM_Additive
    scaleSpace=BCS_BoneSpace
    translationMode=BMM_Additive
    translationSpace=BCS_BoneSpace
    alphaBoolBlend {blendOption=Linear}
    alphaScaleBiasClamp {interpSpeedDecreasing=19.75, interpSpeedIncreasing=19.75}
    boneToModify {boneName=base_bag}
    rotation {pitch=20}
    translation {x=3.5}
   }
   n18 = Blend Poses by bool(
    BlendPose_0=Component To Local(ComponentPose=@n17)
    BlendPose_1=Blend Poses by bool(
     BlendPose_0=Use cached pose 'CTRLRigPose'()
     BlendPose_1=Blend Poses by bool(BlendPose_0=@n14, BlendPose_1=@n10, bActiveValue=@n15) {blendType=Linear}
     bActiveValue=Get bIsParachuteOpened()
    ) {blendType=Linear}
    BlendTime_0=0.4
    BlendTime_1=0.4
    bActiveValue=AND Boolean!(A=Get bIsMoving2D(), B=Get bIsOnGround() -> (bIsOnGround:bool)) -> (ReturnValue)
   ) {blendType=Linear}
   n19 = Get isTargetingHeadUp()
   n20 = "Transform (Modify) Bone - Bone: base_collar_mid_01_l"(
    ComponentPose=Local To Component(LocalPose=Use cached pose 'CTRLRigPose'())
    Alpha=@n19
    Translation="-7.5,-15,3"
    Rotation="23.144,-54.677,-15.330"
    Scale="1,1,1"
   ) {
    rotationMode=BMM_Additive
    rotationSpace=BCS_BoneSpace
    scaleMode=BMM_Additive
    scaleSpace=BCS_BoneSpace
    translationMode=BMM_Additive
    translationSpace=BCS_BoneSpace
    alphaBoolBlend {blendOption=Linear}
    alphaScaleBiasClamp {interpSpeedDecreasing=19.75, interpSpeedIncreasing=19.75}
    boneToModify {boneName=base_collar_mid_01_l}
    rotation=<rot 23.144 -54.677 -15.33>
    translation=<-7.5 -15 3>
   }
   n21 = "Transform (Modify) Bone - Bone: base_collar_side_01_l"(ComponentPose=@n20, Alpha=@n19, Translation="0,-5,0", Rotation="0,0,0", Scale="1,1,1") {
    rotationMode=BMM_Additive
    rotationSpace=BCS_BoneSpace
    scaleMode=BMM_Additive
    scaleSpace=BCS_BoneSpace
    translationMode=BMM_Additive
    translationSpace=BCS_BoneSpace
    alphaBoolBlend {blendOption=Linear}
    alphaScaleBiasClamp {interpSpeedDecreasing=19.75, interpSpeedIncreasing=19.75}
    boneToModify {boneName=base_collar_side_01_l}
    translation {y=-5}
   }
   n22 = "Transform (Modify) Bone - Bone: base_collar_side_01_r"(ComponentPose=@n21, Alpha=@n19, Translation="-5,-20,0", Rotation="-53.704,-133.873,89.597", Scale="1,1,1") {
    rotationMode=BMM_Additive
    rotationSpace=BCS_BoneSpace
    scaleMode=BMM_Additive
    scaleSpace=BCS_BoneSpace
    translationMode=BMM_Additive
    translationSpace=BCS_BoneSpace
    alphaBoolBlend {blendOption=Linear}
    alphaScaleBiasClamp {interpSpeedDecreasing=19.75, interpSpeedIncreasing=19.75}
    boneToModify {boneName=base_collar_side_01_r}
    rotation=<rot -53.704 -133.873 89.597>
    translation=<-5 -20>
   }
   n23 = Component To Local(ComponentPose=@n22)
   n24 = AND Boolean!(A=Get bIsCrouching(), B=Get bIsTargeting(), C:bool=Get bIsMoving2D(), D:bool=Get bIsOnGround()) -> (ReturnValue)
   n25 = "Transform (Modify) Bone - Bone: base_collar_front_01_l"(ComponentPose=Local To Component(LocalPose=Use cached pose 'CTRLRigPose'()), Alpha=1, Translation="-1,0,0", Rotation="0,50,0", Scale="1,1,1") {
    rotationMode=BMM_Additive
    scaleMode=BMM_Additive
    translationMode=BMM_Additive
    alphaBoolBlend {blendOption=Linear}
    boneToModify {boneName=base_collar_front_01_l}
    rotation {yaw=50}
    translation {x=-1}
   }
   n26 = "Transform (Modify) Bone - Bone: dyn_collar_front_02_l"(ComponentPose=@n25, Alpha=1, Translation="0,0,0", Rotation="0,-50,0", Scale="1,1,1") {
    rotationMode=BMM_Additive
    scaleMode=BMM_Additive
    translationMode=BMM_Additive
    alphaBoolBlend {blendOption=Linear}
    boneToModify {boneName=dyn_collar_front_02_l}
    rotation {yaw=-50}
   }
   n27 = "Transform (Modify) Bone - Bone: base_collar_front_01_r"(ComponentPose=@n26, Alpha=1, Translation="1,0,0", Rotation="0,-50,0", Scale="1,1,1") {
    rotationMode=BMM_Additive
    scaleMode=BMM_Additive
    translationMode=BMM_Additive
    alphaBoolBlend {blendOption=Linear}
    boneToModify {boneName=base_collar_front_01_r}
    rotation {yaw=-50}
    translation {x=1}
   }
   n28 = "Transform (Modify) Bone - Bone: dyn_collar_front_02_r"(ComponentPose=@n27, Alpha=1, Translation="0,0,0", Rotation="0,50,0", Scale="1,1,1") {
    rotationMode=BMM_Additive
    scaleMode=BMM_Additive
    translationMode=BMM_Additive
    alphaBoolBlend {blendOption=Linear}
    boneToModify {boneName=dyn_collar_front_02_r}
    rotation {yaw=50}
   }
   Output Pose(Result=Blend Poses by bool(
     BlendPose_0=Component To Local(ComponentPose=@n28)
     BlendPose_1=Blend Poses by bool(BlendPose_0=@n23, BlendPose_1=@n18, BlendTime_0=0.4, BlendTime_1=0.4, bActiveValue=@n24) {blendType=Linear}
     BlendTime_0=0.4
     BlendTime_1=0.4
     bActiveValue=Get bIsPlayingEmote()
    ) {blendType=Linear}) {layerGroup=None}
   n29 = Get headUp() -> (headUp:double)
   n30 = "Transform (Modify) Bone - Bone: base_collar_mid_01_l"(
    ComponentPose=Local To Component(LocalPose=Use cached pose 'CTRLRigPose'())
    Alpha=@n29
    Translation="-5,-20,0"
    Rotation="-14.350,-34.876,-48.342"
    Scale="1,1,1"
   ) {
    rotationMode=BMM_Additive
    rotationSpace=BCS_BoneSpace
    scaleMode=BMM_Additive
    scaleSpace=BCS_BoneSpace
    translationMode=BMM_Additive
    translationSpace=BCS_BoneSpace
    alphaBoolBlend {blendOption=Linear}
    alphaScaleBiasClamp {interpSpeedDecreasing=19.75, interpSpeedIncreasing=19.75}
    boneToModify {boneName=base_collar_mid_01_l}
    rotation=<rot -14.35 -34.876 -48.342>
    translation=<-5 -20>
   }
   n31 = "Transform (Modify) Bone - Bone: base_collar_side_01_l"(ComponentPose=@n30, Alpha=@n29, Translation="0,-10,0", Rotation="0,0,0", Scale="1,1,1") {
    rotationMode=BMM_Additive
    rotationSpace=BCS_BoneSpace
    scaleMode=BMM_Additive
    scaleSpace=BCS_BoneSpace
    translationMode=BMM_Additive
    translationSpace=BCS_BoneSpace
    alphaBoolBlend {blendOption=Linear}
    alphaScaleBiasClamp {interpSpeedDecreasing=19.75, interpSpeedIncreasing=19.75}
    boneToModify {boneName=base_collar_side_01_l}
    translation {y=-10}
   }
   n32 = "Transform (Modify) Bone - Bone: base_collar_side_01_r"(ComponentPose=@n31, Alpha=@n29, Translation="-5,-15,0", Rotation="0.407,-49.995,0.112", Scale="1,1,1") {
    rotationMode=BMM_Additive
    rotationSpace=BCS_BoneSpace
    scaleMode=BMM_Additive
    scaleSpace=BCS_BoneSpace
    translationMode=BMM_Additive
    translationSpace=BCS_BoneSpace
    alphaBoolBlend {blendOption=Linear}
    alphaScaleBiasClamp {interpSpeedDecreasing=19.75, interpSpeedIncreasing=19.75}
    boneToModify {boneName=base_collar_side_01_r}
    rotation=<rot 0.407 -49.995 0.112>
    translation=<-5 -15>
   }
   n33 = "Transform (Modify) Bone - Bone: base_collar_mid_01_l"(ComponentPose=@n32, Alpha=Get IsCrouchingheadup(), Translation="0,5,0", Rotation="0,0,0", Scale="1,1,1") {
    rotationMode=BMM_Additive
    rotationSpace=BCS_BoneSpace
    scaleMode=BMM_Additive
    scaleSpace=BCS_BoneSpace
    translationMode=BMM_Additive
    translationSpace=BCS_BoneSpace
    alphaBoolBlend {blendOption=Linear}
    alphaScaleBiasClamp {interpSpeedDecreasing=19.75, interpSpeedIncreasing=19.75}
    boneToModify {boneName=base_collar_mid_01_l}
    translation {y=5}
   }
   Save cached pose 'HeadUpPoses'(Pose=Component To Local(ComponentPose=@n33)) {cacheName=HeadUpPoses}
   n34 = Control Rig!(
    Source=Copy Pose From Mesh() {bUseAttachedParent=true}
    Alpha
    get is crouching bag:double=Get IsCrouchingBag()
    get is skydiving alpha:double=Get Is Skydiving Alpha() -> (Is Skydiving Alpha:double)
    get is head rotation limit alpha:double=0
   ) -> (Pose) {
    alphaBoolBlend {blendOption=Linear}
    controlRigAssetReference {blueprintRigClass=&/M_MED_RankedSpeeder_DeformRig.M_MED_RankedSpeeder_DeformRig_C}
   }
   n35 = "Transform (Modify) Bone - Bone: base_bag"(
    ComponentPose=Local To Component(LocalPose=@n34)
    Alpha=Get IsCrouchingBag()
    Translation="6,0,0"
    Rotation="-10.000,0.045,-0.016"
    Scale="1,1,1"
   ) {
    rotationMode=BMM_Additive
    rotationSpace=BCS_BoneSpace
    scaleMode=BMM_Additive
    scaleSpace=BCS_BoneSpace
    translationMode=BMM_Additive
    translationSpace=BCS_BoneSpace
    alphaBoolBlend {blendOption=Linear}
    alphaScaleBiasClamp {interpSpeedDecreasing=19.75, interpSpeedIncreasing=19.75}
    boneToModify {boneName=base_bag}
    rotation=<rot -10 0.045 -0.016>
    translation {x=6}
   }
   n36 = "Transform (Modify) Bone - Bone: attach_backpack"!(ComponentPose=@n35, Alpha=1, Translation="0,1.842,0") -> (Pose) {
    translationMode=BMM_Additive
    translationSpace=BCS_BoneSpace
    alphaBoolBlend {blendOption=Linear}
    boneToModify {boneName=attach_backpack}
    translation {y=1.842}
   }
   n37 = "Transform (Modify) Bone - Bone: dyn_legPad_l"!(ComponentPose=@n36, Alpha=Get crouch_legpad_clipping() -> (crouch_legpad_clipping:double), Translation="-1.2,0,0") -> (Pose) {
    translationMode=BMM_Additive
    translationSpace=BCS_BoneSpace
    alphaBoolBlend {blendOption=Linear}
    boneToModify {boneName=dyn_legPad_l}
    translation {x=-1.2}
   }
   n38 = RigidBody!(ComponentPose=@n37, Alpha, OverrideWorldGravity:Vector=Get GravityOverride()) -> (Pose) {
    bClampLinearTranslationLimitToRefPose=true
    bOverrideWorldGravity=true
    overridePhysicsAsset=&/M_MED_RankedSpeeder_Physics
    simulationSpace=BaseBoneSpace
    alphaBoolBlend {blendOption=Linear}
    baseBoneRef {boneName=spine_02}
    componentAppliedLinearAccClamp=<3000 3000 3000>
    componentLinearAccScale=<0.4 0.3 0.3>
    componentLinearVelScale=<0.2 0.2 0.2>
    simSpaceSettings {
     maxAngularAcceleration=1000
     maxAngularVelocity=3
     maxLinearAcceleration=700
     maxLinearVelocity=500
     velocityScaleZ=0
     worldAlpha=0.2
     externalLinearDragV=<0.2 0.2>
    }
   }
   Save cached pose 'CTRLRigPose'(Pose=Component To Local(ComponentPose=@n38)) {cacheName=CTRLRigPose}
   Get HeadLimitAlpha() -> (HeadLimitAlpha:double)
  ]}
 EventGraph {nodes [
   n1 = Event BlueprintUpdateAnimation()
   > n2 = Is Valid(InputObject=TryGetPawnOwner())
   n3 = Set CinePosing(execute=@n2.Is Not Valid, CinePosing:bool=true) -> (Output_Get:bool)
   n4 = Get bIsPlayingEmote()
   n5 = Get CinePosing() -> (CinePosing:bool)
   Set HeadLimit(execute=[@n3.then, @n8.then], HeadLimit:bool=AND Boolean!(A=NOT Boolean(A=@n4), B=false, C=NOT Boolean(A=@n5)) -> (ReturnValue)) -> (Output_Get:bool)
   > Set HeadLimitAlpha(HeadLimitAlpha:double="To Float (Boolean)"(InBool=Get HeadLimit() -> (HeadLimit:bool))) -> (Output_Get:double)
   n6 = NOT Boolean(A=Get bIsPlayingEmote())
   n7 = Get bIsInFrontEnd()
   > Set Is in Front End Alpha(Is in Front End Alpha:double="To Float (Boolean)"(InBool=AND Boolean!(A=@n7, B=@n6) -> (ReturnValue))) -> (Output_Get:double)
   n8 = Set CinePosing(execute=@n2.Is Valid, CinePosing=false)
   n9 = Sequence!(execute) -> (then_0, then_1, then_2:Exec, then_3:Exec, then_4:Exec)
   > n10 = Gravity Override State Function DEPRECATED(CharacterPartAnimInstance=Self-Reference(), GameTimeInSeconds=GetGameTimeInSeconds(), GravityOverrideParams=Get Gravity Override Params())
   > Set GravityOverride(GravityOverride=@n10)
   > Set SkelMesh(SkelMesh=GetPartSkeletalMeshComponent())
   > Set IsCrouchingBag(IsCrouchingBag:double="To Float (Boolean)"(InBool=Get bIsCrouching())) -> (Output_Get:double)
   n11 = Get bIsCrouching()
   > Set IsCrouchingheadup(IsCrouchingheadup:double=float * float(A="To Float (Boolean)"(InBool=@n11), B=Get headUp()) {is=title K2Node_PromotableOperator}) -> (Output_Get:double)
   n12 = Get bIsTargeting()
   n13 = Get bIsCrouching()
   > Set CrouchingTargeting(CrouchingTargeting:double=float * float(A="To Float (Boolean)"(InBool=@n13), B="To Float (Boolean)"(InBool=@n12))) -> (Output_Get:double)
   Set Is Skydiving Alpha(execute=@n9.then_3, Is Skydiving Alpha:double="To Float (Boolean)"(InBool=Get bIsSkyDiving())) -> (Output_Get:double)
   n14 = Set Skel Mesh(execute=@n9.then_4, Skel Mesh=GetOwnerSkeletalMeshComponent())
   > Set crouch_legpad_clipping(crouch_legpad_clipping:double=GetAlphaFromReferenceTransformDelta(SkeletalMeshComponent=@n14, BoneName=thigh_r, BaseName=pelvis, InputRange=(X=0,Y=-65), TwistAxis="0,0,1")) -> (Output_Get:double)
   n15 = Get bIsTargeting()
   Set isTargeting(
    execute=@n9.then_1
    isTargeting:double=FInterpTo(Current=Get isTargeting() -> (isTargeting:double), Target="To Float (Boolean)"(InBool=@n15), DeltaTime=@n1.DeltaTimeX, InterpSpeed=8)
   ) -> (Output_Get:double)
   > Set Is Targeting headUp Back(Is Targeting headUp Back:double=float * float!(A=Get isTargeting(), B=Get headUp(), C:double=Get Backward() -> (Backward:double), D:bool=Get bIsCrouching()) -> (ReturnValue)) -> (Output_Get:double)
   n16 = NOT Boolean(A=Get bIsMoving2D())
   n17 = Get bIsCrouching()
   n18 = Get isUsingDoublePistols() -> (isUsingDoublePistols:bool)
   > Set isUsingDoublePistolsCrouch(isUsingDoublePistolsCrouch:double="To Float (Boolean)"(InBool=AND Boolean!(A=@n18, B=@n17, C:bool=@n16) -> (ReturnValue))) -> (Output_Get:double)
   > Set isTargetingHeadUp(isTargetingHeadUp:double=float * float!(A=Get isTargeting(), B=Get headUp()) -> (ReturnValue)) -> (Output_Get:double)
   > Set headUp(headUp:double=GetAlphaFromReferenceTransformDelta(SkeletalMeshComponent=Get SkelMesh(), BoneName=spine_05, BaseName=head, InputRange=(X=0,Y=90), TwistAxis="0,0,1")) -> (Output_Get:double)
   > Set isUsingRocketLauncherFloat(isUsingRocketLauncherFloat:double="To Float (Boolean)"(InBool=Get isUsingRocketLauncher() -> (isUsingRocketLauncher:bool))) -> (Output_Get:double)
   n19 = AND Boolean!(A=Get bIsCrouching(), B=Get bIsMoving2D(), C=Get isUsingPistol() -> (isUsingPistol:bool), D:bool=Get bIsTargeting()) -> (ReturnValue)
   > Set isCrounchingMoving_Pistol(isCrounchingMoving_Pistol:double=float * float(A="To Float (Boolean)"(InBool=@n19), B=Get headUp())) -> (Output_Get:double)
   > n20 = Branch(Condition=Get bIsCrouching())
   > Set translateCollarBack(translateCollarBack:Vector="-5,0,0") -> (Output_Get:Vector)
   > Set rotationCollarBack(rotationCollarBack:Rotator="25,0,0") -> (Output_Get:Rotator)
   Set translateCollarBack(execute=@n20.else, translateCollarBack="0,-5,0")
   > Set rotationCollarBack(rotationCollarBack="0,-90,0")
   n21 = PlayerMovement(execute=@n9.then_2, AnimInstance=Self-Reference(), XDotProductCoefficient=100, YDotProductCoefficient=100, ZDotProductCoefficient=100)
   n22 = BreakVector(InVec=MakeVector(X=@n21, Y=@n21.YMovement))
   > Set Left(Left:double="To Float (Boolean)"(InBool="float < float"(A=@n22, B=-50))) -> (Output_Get:double)
   > Set Right(Right:double="To Float (Boolean)"(InBool="float > float"(A=@n22, B=50))) -> (Output_Get:double)
   > Set Forward(Forward:double="To Float (Boolean)"(InBool="float > float"(A=@n22.Y, B=40))) -> (Output_Get:double)
   > Set Backward(Backward:double="To Float (Boolean)"(InBool="float < float"(A=@n22.Y, B=-40))) -> (Output_Get:double)
  ]}
]
variables [
 Backward {type=number}
 CinePosing {type=boolean}
 crouch_legpad_clipping {type=number}
 CrouchingTargeting {type=number}
 Forward {type=number}
 Gravity Override Params {type=GravityOverrideParamsStruct}
 GravityOverride {type=Vector}
 HeadLimit {type=boolean}
 HeadLimitAlpha {type=number}
 headUp {type=number}
 Is in Front End Alpha {type=number}
 Is Skydiving Alpha {type=number}
 Is Targeting headUp Back {type=number}
 IsCrouchingBag {type=number}
 IsCrouchingheadup {type=number}
 isCrounchingMoving_Pistol {type=number}
 isTargeting {type=number}
 isTargetingHeadUp {type=number}
 isUsingDoublePistols {type=boolean}
 isUsingDoublePistolsCrouch {type=number}
 isUsingPistol {type=boolean}
 isUsingRocketLauncher {type=boolean}
 isUsingRocketLauncherFloat {type=number}
 Left {type=number}
 Right {type=number}
 rotationCollarBack {type=Rotator}
 Skel Mesh {type=SkeletalMeshComponent}
 SkelMesh {type=SkeletalMeshComponent}
 translateCollarBack {type=Vector}
]
